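# Unlit, masked holographic material for the Ultrabarro vehicle wheel, built on the shared simple-alpha wheel parent function.
# Fresnel from the wheel normals, rim and tread masks, and a desaturated or boosted PaintColor build the glow.
# A panning gradient adds travelling waves to the emissive.
# Computed base color and emissive override the parent's attributes, which then pass through vehicle damage and de-rez transition functions.
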Material /VehicleCosmetics/Wheels/UltrabarroHolographic/Materials/M_Wheel_Ultrabarro_Holographic
graph {
 properties {
  bAutomaticallySetUsageInEditor=false
  bUseMaterialAttributes=true
  bUsedWithMeshDeformer=true
  bUsedWithSkeletalMesh=true
  blendMode=BLEND_Masked
  pixelDepthOffsetMode=PDOM_Legacy
  previewMesh=/Engine/Transient.MaterialEditorMeshComponent_2
  shadingModel=MSM_Unlit
  usedShadingModels=Unlit
 }
 nodes [
  n1 = VertexColor()
  Add(A=Constant(), B=@n1.R)
  n2 = MaterialFunctionCall() {materialFunction=/VehicleCosmetics/Wheels/Mage/Materials/MF_WheelParent_Simple_Alpha}
  n3 = NamedRerouteUsage()
  n4 = NamedRerouteUsage()
  MaterialFunctionCall(In=MaterialFunctionCall(Material=SetMaterialAttributes!(MaterialAttributes=@n2.MaterialAttributes, Base Color=@n3, Emissive Color=@n4) {attributeSetTypes [#1, #2]}) {materialFunction=&MF_BaseVehicle_Damage}.Result) {materialFunction=&MF_VehicleDeRez_Transition}
  n5 = TextureSampleParameter2D() {
   desc=CHANGE TO REROUT
   parameterName=Tire_Normal
   samplerType=SAMPLERTYPE_Normal
   texture=/VehicleCosmetics/Wheels/UltrabarroHolographic/Textures/T_Ultrabarro_Tread_Normal
  }
  n6 = Max(A=Abs(None=@n5.R), B=Abs(None=@n5.G))
  n7 = Clamp(None=ConstantBiasScale(None=Clamp(None=@n6)) {bias=-0.1, scale=5})
  n8 = OneMinus(None=Sine(None=Time()) {period=2})
  n9 = Add(A=Multiply(A=Clamp(None=@n8), B=ScalarParameter() {defaultValue=3, parameterName=WaveGlowIntensity}), B=Constant())
  n10 = LinearInterpolate(A=Constant(), B=NamedRerouteUsage(), Alpha=NamedRerouteUsage())
  Max(B=LinearInterpolate(A=Multiply(A=@n9, B=@n10), B=Multiply(A=@n7, B=@n9), Alpha=NamedRerouteUsage()))
  NamedRerouteDeclaration(None=@n2.NormalTire) {name=Normal_Tire_IN, nodeColor=<rgba 0.2951 0.1556 1 1>}
  n11 = TextureSampleParameter2D() {
   group=Shader_Textures
   parameterName=RIM_Mask
   sortPriority=1
   texture=/VehicleCosmetics/Wheels/UltrabarroHolographic/Textures/T_Wheel_UltrabarroHolographic_Rim_RGB
  }
  NamedRerouteDeclaration(None=@n11.R) {name=RIM_R, nodeColor=<rgba 1 0 0 1>}
  n12 = NamedRerouteUsage()
  n13 = MaterialFunctionCall!(
   Normal=LinearInterpolate(A=NamedRerouteUsage(), B=NamedRerouteUsage(), Alpha=@n12)
   Exponent=ScalarParameter() {defaultValue=3, parameterName=FresnelExponent}
  ) {materialFunction=/VehicleCosmetics/Wheels/Cola_Holographic/Materials/MF_Fresnel_Classic}
  n14 = Multiply(A=ConstantBiasScale(None=@n12) {bias=0.2, scale=4}, B=Constant() {r=1})
  n15 = ConstantBiasScale(None=Multiply(A=ScalarParameter() {defaultValue=0.15, parameterName=EmissiveIntensity}, B=Multiply(A=@n13.Result, B=@n14))) {bias=0.02, scale=1}
  n16 = Add(A=NamedRerouteUsage(), B=Multiply(A=NamedRerouteUsage(), B=@n15))
  n17 = VectorParameter() {parameterName=PaintColor, defaultValue=<rgba 0 0.0789 0.3438 1>}
  n18 = Constant() {r=10}
  n19 = TextureSample(UVs=Panner(Coordinate=TextureCoordinate()) {speedY=-0.5}) {texture=/VehicleCosmetics/Wheels/UltrabarroHolographic/Textures/SharpGradient_01_Heightmap}
  n20 = OneMinus(None=@n19.R)
  n21 = Clamp(None=ConstantBiasScale(None=Clamp(None=@n20)) {bias=-0.9, scale=0.25})
  n22 = Multiply(
   A=LinearInterpolate(A=NamedRerouteUsage(), B=NamedRerouteUsage(), Alpha=NamedRerouteUsage())
   B=Multiply(A=Multiply(A=@n18, B=@n21), B=ScalarParameter() {defaultValue=3, parameterName=WaveGlowIntensity})
  )
  NamedRerouteDeclaration(None=Max(A=@n16, B=Multiply(A=@n17.RGB, B=@n22))) {name=Emissive, nodeColor=<rgba 1 0 0.8394 1>}
  NamedRerouteDeclaration(None=@n16) {name=BC, nodeColor=<rgba 0 0.1492 1 1>}
  n23 = TextureSampleParameter2D() {
   group=Shader_Textures
   parameterName=TIRE_Mask
   sortPriority=1
   texture=/VehicleCosmetics/Wheels/UltrabarroHolographic/Textures/Ultrabarro_Tread_M2
  }
  NamedRerouteDeclaration(None=@n23.B) {name=TIRE_B, nodeColor=<rgba 0 0 1 1>}
  NamedRerouteDeclaration(None=@n1.R) {name=VC_R0, nodeColor=<rgba 1 0 0.096 1>}
  n24 = VectorParameter() {parameterName=PaintColor, defaultValue=<rgba 0 0.0789 0.3438 1>}
  n25 = Desaturation(None=@n24.RGB, Fraction=Constant())
  n26 = Constant() {r=5}
  n27 = NamedRerouteUsage()
  n28 = NamedRerouteUsage()
  n29 = NamedRerouteUsage()
  n30 = Multiply(A=ScalarParameter() {defaultValue=3, parameterName=WaveGlowIntensity}, B=Multiply(A=LinearInterpolate(A=@n27, B=@n28, Alpha=@n29), B=@n19.R))
  n31 = Constant()
  n32 = NamedRerouteUsage()
  NamedRerouteDeclaration(None=LinearInterpolate(A=@n25, B=Multiply(A=@n24.RGB, B=@n26), Alpha=LinearInterpolate(A=@n30, B=@n31, Alpha=@n32))) {name=Desat_BC, nodeColor=<rgba 0.243 0.3153 1 1>}
  n33 = Desaturation(None=@n25, Fraction=Constant() {r=0.5})
  n34 = Multiply(A=Multiply(A=@n33, B=@n33), B=NamedRerouteUsage())
  n35 = Constant() {r=0.5}
  NamedRerouteDeclaration(None=Clamp(None=Multiply(A=@n34, B=@n35))) {name=Desat_BC_Clamp, nodeColor=<rgba 0 0.2523 1 1>}
  NamedRerouteDeclaration(None=@n2.NormalRim) {name=Normal_Rim_IN, nodeColor=<rgba 0 1 0.2969 1>}
  NamedRerouteDeclaration(None=VertexColor().R) {name=VC_R, nodeColor=<rgba 1 0 0.0487 1>}
  NamedRerouteDeclaration(None=@n11.G) {name=RIM_G, nodeColor=<rgba 0 1 0 1>}
  NamedRerouteDeclaration(None=@n11.B) {name=RIM_B, nodeColor=<rgba 0 0 1 1>}
  NamedRerouteDeclaration(None=@n23.R) {name=TIRE_R, nodeColor=<rgba 1 0 0 1>}
  NamedRerouteDeclaration(None=@n23.G) {name=TIRE_G, nodeColor=<rgba 0 1 0 1>}
  Subtract(A=@n15)
 ]
}
