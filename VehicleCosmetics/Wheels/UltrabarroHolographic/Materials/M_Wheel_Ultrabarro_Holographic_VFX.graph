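# Translucent unlit material for the holographic effect on the Ultrabarro vehicle wheel.
# Opacity sums six blurred-line layers panning at different speeds, masks them with scrolling grunge gradients, then uses scanline texture alpha to blend circle- and stripe-masked versions.
# Emissive boosts PaintColor, optionally mixed with a hue-shifted copy by camera-vector dot products for view-dependent color.
# The attributes pass through the shared vehicle damage and de-rez transition functions.
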Material /VehicleCosmetics/Wheels/UltrabarroHolographic/Materials/M_Wheel_Ultrabarro_Holographic_VFX
graph {
 properties {
  bAutomaticallySetUsageInEditor=false
  bUseMaterialAttributes=true
  bUsedWithMeshDeformer=true
  bUsedWithSkeletalMesh=true
  blendMode=BLEND_Translucent
  previewMesh=/VehicleCosmetics/Wheels/Esports01/Meshes/SK_Wheel_ESports01
  shadingModel=MSM_Unlit
  usedShadingModels=Unlit
 }
 nodes [
  n1 = NamedRerouteUsage()
  n2 = NamedRerouteUsage()
  MaterialFunctionCall(In=MaterialFunctionCall(Material=SetMaterialAttributes!(MaterialAttributes, Emissive Color=@n1, Opacity=@n2) {attributeSetTypes [#1, #2]}) {materialFunction=&MF_BaseVehicle_Damage}.Result) {materialFunction=&MF_VehicleDeRez_Transition}
  n3 = Constant() {r=1}
  n4 = TextureCoordinate()
  n5 = TextureSample(UVs=Rotator(Coordinate=Panner(Coordinate=@n4) {speedY=0.01}, Time=Constant() {r=0.0175}) {speed=90}) {texture=/VehicleCosmetics/Wheels/UltrabarroHolographic/Textures/T_Gradient_Grunges}
  n6 = Constant() {r=1}
  n7 = TextureSample(UVs=Panner(Coordinate=TextureCoordinate() {uTiling=2}) {speedY=0.015}) {texture=&T_Gradient_Grunges}
  n8 = Rotator(Coordinate=Panner(Coordinate=TextureCoordinate() {vTiling=23}) {speedY=-0.15}, Time=Constant() {r=0.0175}) {speed=90}
  n9 = TextureCoordinate() {vTiling=22}
  n10 = TextureSample(UVs=Rotator(Coordinate=Panner(Coordinate=@n9) {speedY=-0.189}, Time=Constant() {r=0.0175}) {speed=90}) {texture=/VehicleCosmetics/Wheels/UltrabarroHolographic/Textures/T_Blurredline}
  n11 = Panner(Coordinate=TextureCoordinate() {vTiling=26}) {speedY=-0.23}
  n12 = Constant() {r=0.0175}
  n13 = Rotator(Coordinate=Panner(Coordinate=TextureCoordinate() {vTiling=20}) {speedY=-0.248}, Time=Constant() {r=0.0175}) {speed=90}
  n14 = TextureCoordinate() {vTiling=30}
  n15 = TextureSample(UVs=Rotator(Coordinate=Panner(Coordinate=@n14) {speedY=-0.168}, Time=Constant() {r=0.0175}) {speed=90}) {texture=&T_Blurredline}
  n16 = TextureCoordinate() {vTiling=25}
  n17 = TextureSample(UVs=Rotator(Coordinate=Panner(Coordinate=@n16) {speedY=-0.2}, Time=Constant() {r=0.0175}) {speed=90}) {texture=&T_Blurredline}
  n18 = Add(
   A=TextureSample(UVs=Rotator(Coordinate=@n11, Time=@n12) {speed=90}) {texture=&T_Blurredline}.RGB
   B=Add(A=TextureSample(UVs=@n13) {texture=&T_Blurredline}.RGB, B=Add(A=@n15.RGB, B=@n17.RGB))
  )
  n19 = Multiply(
   A=Multiply(A=Multiply(A=@n3, B=@n5.B), B=Multiply(A=@n6, B=@n7.B))
   B=Add(A=TextureSample(UVs=@n8) {texture=&T_Blurredline}.RGB, B=Add(A=@n10.RGB, B=@n18))
  )
  n20 = TextureSample(UVs=TextureCoordinate() {uTiling=3, vTiling=3}) {
   samplerSource=SSM_Wrap_WorldGroupSettings
   samplerType=SAMPLERTYPE_Grayscale
   texture=/VehicleCosmetics/Wheels/_SharedTextures/FX/GradientCircle01
  }
  n21 = Multiply(A=Multiply(A=@n19, B=Constant() {r=1}), B=@n20.RGB)
  n22 = TextureSample() {texture=/VehicleCosmetics/Wheels/UltrabarroHolographic/Textures/T_Waves_Scanlines}
  n23 = TextureSample(UVs=Rotator(Coordinate=TextureCoordinate() {uTiling=25}, Time=Constant() {r=0.0174}) {speed=90}) {samplerSource=SSM_Clamp_WorldGroupSettings, texture=/VehicleCosmetics/Wheels/_SharedTextures/FX/GradientStripe_T}
  n24 = Add(A=Multiply(A=@n19, B=@n23.RGB), B=Multiply(A=@n23.RGB, B=Constant() {r=0.1}))
  NamedRerouteDeclaration(None=Multiply(A=LinearInterpolate(A=@n21, B=@n24, Alpha=@n22.A), B=ScalarParameter() {defaultValue=2, parameterName=Opacity})) {name=Opacity, nodeColor=<rgba 0.5021 0.567 0.5781 1>}
  n25 = VectorParameter() {parameterName=PaintColor, defaultValue=<rgba 0.04 0 0.65 1>}
  n26 = Constant() {r=0.2}
  n27 = CameraVectorWS()
  n28 = OneMinus(None=@n27)
  n29 = MaterialFunctionCall(Hue Shift Percentage=ScalarParameter() {defaultValue=-0.2, parameterName=HueShift}, Texture=@n25.RGB) {materialFunction=&HueShift}
  n30 = Constant() {r=1}
  n31 = Constant() {r=0.1}
  n32 = Add(A=Multiply(A=@n25.RGB, B=DotProduct(A=@n26, B=@n28)), B=Multiply(A=Multiply(A=@n29.Result, B=@n30), B=DotProduct(A=@n31, B=@n27)))
  NamedRerouteDeclaration(None=Multiply(A=StaticSwitchParameter(True=@n32, False=@n25.RGB) {defaultValue=true, parameterName=MultiEmissiveBlend}, B=Constant() {r=8})) {name=Emissive, nodeColor=<rgba 1 0 0.7774 1>}
  Reroute(None=@n24)
 ]
}
